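# Sound cue for one-shot Blue Jay bird calls, an outdoor ambient spot sound for the Olympus biome.
# A random node picks a triple loop, a double loop or a single call, each drawing from the same three wave variations.
# The result passes through a Modulator for variation, which feeds all quality levels identically.
# Bird-specific attenuation, a low volume multiplier and a distance-based exterior ambient bus send shape playback.
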SoundCue /Game/Sounds/Helios/Biomes/Olympus/Spots/Birds/Spot_Bird_Olympus_BlueJay_Cue
root=QualityLevel_0
nodes [
 {id=Looping_0, children [Random_1], properties {bLoopIndefinitely=false, loopCount=3}}
 {id=Looping_2, children [Random_4], properties {bLoopIndefinitely=false, loopCount=2}}
 {id=Modulator_0, children [Random_6]}
 {id=QualityLevel_0, children [Modulator_0,Modulator_0,Modulator_0,Modulator_0]}
 {id=Random_1, children [WavePlayer_16, WavePlayer_18, WavePlayer_19], properties {weights [1,1,1]}}
 {id=Random_4, children [WavePlayer_34, WavePlayer_35, WavePlayer_36], properties {weights [1,1,1]}}
 {id=Random_5, children [WavePlayer_37, WavePlayer_38, WavePlayer_39], properties {weights [1,1,1]}}
 {id=Random_6, children [Looping_0,Looping_2,Random_5], properties {weights [1,1,1]}}
 {id=WavePlayer_16, properties {soundWaveAssetPtr=&/Spot_Bird_BlueJay_03}}
 {id=WavePlayer_18, properties {soundWaveAssetPtr=&/Spot_Bird_BlueJay_02}}
 {id=WavePlayer_19, properties {soundWaveAssetPtr=&/Spot_Bird_BlueJay_01}}
 {id=WavePlayer_34, properties {soundWaveAssetPtr=&Spot_Bird_BlueJay_03}}
 {id=WavePlayer_35, properties {soundWaveAssetPtr=&Spot_Bird_BlueJay_02}}
 {id=WavePlayer_36, properties {soundWaveAssetPtr=&Spot_Bird_BlueJay_01}}
 {id=WavePlayer_37, properties {soundWaveAssetPtr=&Spot_Bird_BlueJay_03}}
 {id=WavePlayer_38, properties {soundWaveAssetPtr=&Spot_Bird_BlueJay_02}}
 {id=WavePlayer_39, properties {soundWaveAssetPtr=&Spot_Bird_BlueJay_01}}
]
properties {
 attenuationSettings=&Fort_TileAmbient_OneShot_Bird_Att
 maxDistance=8005
 soundClassObject=&Fort_Ambient_OneShots_Outdoors
 volumeMultiplier=0.2
 busSends [{
   audioBus=null
   maxSendDistance=6000
   minSendDistance=3000
   minSendLevel=0.6
   sendLevel=1
   soundSourceBus=&AmbientAudio_Exterior_SourceBus
  }]
}
