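# Sound cue for one-shot Greek mountain bird calls, an outdoor ambient spot in the Helios Olympus biome.
# A quality-level switch uses five wave variations in its first two slots and only two in the last two.
# Each branch passes through a Modulator into a weighted random choice of a single call or two or three repeats with random delays.
# Bird one-shot attenuation applies, and a distance-based send feeds the exterior ambient source bus.
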SoundCue /Game/Sounds/Helios/Biomes/Olympus/Spots/Birds/Spot_Bird_Olympus_Greek_Mountain_Cue
root=QualityLevel_0
nodes [
 {id=Delay_0, children [Random_0], properties {delayMax=2}}
 {id=Delay_1, children [Random_7], properties {delayMax=2}}
 {id=Delay_2, children [Random_2], properties {delayMax=2}}
 {id=Delay_3, children [Random_9], properties {delayMax=2}}
 {id=Looping_0, children [Delay_0], properties {bLoopIndefinitely=false, loopCount=3}}
 {id=Looping_1, children [Delay_1], properties {bLoopIndefinitely=false, loopCount=3}}
 {id=Looping_2, children [Delay_2], properties {bLoopIndefinitely=false, loopCount=2}}
 {id=Looping_3, children [Delay_3], properties {bLoopIndefinitely=false, loopCount=2}}
 {id=Modulator_0, children [Random_6]}
 {id=Modulator_1, children [Random_8]}
 {id=QualityLevel_0, children [Modulator_0,Modulator_0,Modulator_1,Modulator_1]}
 {id=Random_0, children [WavePlayer_0,WavePlayer_1,WavePlayer_2,WavePlayer_3,WavePlayer_4], properties {weights [1,1,1,1,1]}}
 {id=Random_10, children [WavePlayer_24, WavePlayer_25], properties {weights [1,1]}}
 {id=Random_2, children [WavePlayer_5,WavePlayer_6,WavePlayer_7,WavePlayer_8,WavePlayer_9], properties {weights [1,1,1,1,1]}}
 {id=Random_3, children [WavePlayer_10, WavePlayer_11, WavePlayer_12, WavePlayer_13, WavePlayer_14], properties {weights [1,1,1,1,1]}}
 {id=Random_6, children [Looping_0,Looping_2,Random_3], properties {weights [0.5,0.75,1]}}
 {id=Random_7, children [WavePlayer_20, WavePlayer_21], properties {weights [1,1]}}
 {id=Random_8, children [Looping_1,Looping_3,Random_10], properties {weights [0.5,1,0.75]}}
 {id=Random_9, children [WavePlayer_22, WavePlayer_23], properties {weights [1,1]}}
 {id=WavePlayer_0, properties {soundWaveAssetPtr=&/Spot_Bird_Greek_Mountain_01}}
 {id=WavePlayer_1, properties {soundWaveAssetPtr=&/Spot_Bird_Greek_Mountain_02}}
 {id=WavePlayer_10, properties {soundWaveAssetPtr=&Spot_Bird_Greek_Mountain_01}}
 {id=WavePlayer_11, properties {soundWaveAssetPtr=&Spot_Bird_Greek_Mountain_02}}
 {id=WavePlayer_12, properties {soundWaveAssetPtr=&/Spot_Bird_Greek_Mountain_03}}
 {id=WavePlayer_13, properties {soundWaveAssetPtr=&/Spot_Bird_Greek_Mountain_04}}
 {id=WavePlayer_14, properties {soundWaveAssetPtr=&/Spot_Bird_Greek_Mountain_05}}
 {id=WavePlayer_2, properties {soundWaveAssetPtr=&Spot_Bird_Greek_Mountain_03}}
 {id=WavePlayer_20, properties {soundWaveAssetPtr=&Spot_Bird_Greek_Mountain_01}}
 {id=WavePlayer_21, properties {soundWaveAssetPtr=&Spot_Bird_Greek_Mountain_02}}
 {id=WavePlayer_22, properties {soundWaveAssetPtr=&Spot_Bird_Greek_Mountain_01}}
 {id=WavePlayer_23, properties {soundWaveAssetPtr=&Spot_Bird_Greek_Mountain_02}}
 {id=WavePlayer_24, properties {soundWaveAssetPtr=&Spot_Bird_Greek_Mountain_01}}
 {id=WavePlayer_25, properties {soundWaveAssetPtr=&Spot_Bird_Greek_Mountain_02}}
 {id=WavePlayer_3, properties {soundWaveAssetPtr=&Spot_Bird_Greek_Mountain_04}}
 {id=WavePlayer_4, properties {soundWaveAssetPtr=&Spot_Bird_Greek_Mountain_05}}
 {id=WavePlayer_5, properties {soundWaveAssetPtr=&Spot_Bird_Greek_Mountain_01}}
 {id=WavePlayer_6, properties {soundWaveAssetPtr=&Spot_Bird_Greek_Mountain_02}}
 {id=WavePlayer_7, properties {soundWaveAssetPtr=&Spot_Bird_Greek_Mountain_03}}
 {id=WavePlayer_8, properties {soundWaveAssetPtr=&Spot_Bird_Greek_Mountain_04}}
 {id=WavePlayer_9, properties {soundWaveAssetPtr=&Spot_Bird_Greek_Mountain_05}}
]
properties {
 attenuationSettings=&Fort_TileAmbient_OneShot_Bird_Att
 maxDistance=8005
 soundClassObject=&Fort_Ambient_OneShots_Outdoors
 volumeMultiplier=0.251
 busSends [{
   audioBus=null
   maxSendDistance=6000
   minSendDistance=3000
   minSendLevel=0.6
   sendLevel=1
   soundSourceBus=&AmbientAudio_Exterior_SourceBus
  }]
}
