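# Sound cue for a Greenish Warbler spot bird call, an outdoor ambient one-shot in the Olympus biome with bird attenuation and an exterior source bus send.
# A quality switch sends its first two levels to a five-wave branch and its last two to a reduced two-wave branch.
# Each branch modulates a weighted random pick between a single call and two or three repeats, each repeat randomly chosen and delayed.
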SoundCue /Game/Sounds/Helios/Biomes/Olympus/Spots/Birds/Spot_Bird_Olympus_Greenish_Warbler_Cue
root=QualityLevel_0
nodes [
 {id=Delay_0, children [Random_1], properties {delayMax=2}}
 {id=Delay_1, children [Random_7], properties {delayMax=2}}
 {id=Delay_2, children [Random_4], properties {delayMax=2}}
 {id=Delay_3, children [Random_9], properties {delayMax=2}}
 {id=Looping_0, children [Delay_0], properties {bLoopIndefinitely=false, loopCount=3}}
 {id=Looping_1, children [Delay_1], properties {bLoopIndefinitely=false, loopCount=3}}
 {id=Looping_2, children [Delay_2], properties {bLoopIndefinitely=false, loopCount=2}}
 {id=Looping_3, children [Delay_3], properties {bLoopIndefinitely=false, loopCount=2}}
 {id=Modulator_0, children [Random_6]}
 {id=Modulator_1, children [Random_8]}
 {id=QualityLevel_0, children [Modulator_0,Modulator_0,Modulator_1,Modulator_1]}
 {id=Random_1, children [WavePlayer_15, WavePlayer_16, WavePlayer_17, WavePlayer_18, WavePlayer_19], properties {weights [1,1,1,1,1]}}
 {id=Random_10, children [WavePlayer_40, WavePlayer_41], properties {weights [1,1]}}
 {id=Random_4, children [WavePlayer_26, WavePlayer_27, WavePlayer_28, WavePlayer_29, WavePlayer_30], properties {weights [1,1,1,1,1]}}
 {id=Random_5, children [WavePlayer_31, WavePlayer_32, WavePlayer_33, WavePlayer_34, WavePlayer_35], properties {weights [1,1,1,1,1]}}
 {id=Random_6, children [Looping_0,Looping_2,Random_5], properties {weights [0.5,0.75,1]}}
 {id=Random_7, children [WavePlayer_36, WavePlayer_37], properties {weights [1,1]}}
 {id=Random_8, children [Looping_1,Looping_3,Random_10], properties {weights [0.5,1,0.75]}}
 {id=Random_9, children [WavePlayer_38, WavePlayer_39], properties {weights [1,1]}}
 {id=WavePlayer_15, properties {soundWaveAssetPtr=&/Spot_Bird_Greenish_Warbler_01}}
 {id=WavePlayer_16, properties {soundWaveAssetPtr=&/Spot_Bird_Greenish_Warbler_02}}
 {id=WavePlayer_17, properties {soundWaveAssetPtr=&/Spot_Bird_Greenish_Warbler_03}}
 {id=WavePlayer_18, properties {soundWaveAssetPtr=&/Spot_Bird_Greenish_Warbler_04}}
 {id=WavePlayer_19, properties {soundWaveAssetPtr=&/Spot_Bird_Greenish_Warbler_05}}
 {id=WavePlayer_26, properties {soundWaveAssetPtr=&Spot_Bird_Greenish_Warbler_01}}
 {id=WavePlayer_27, properties {soundWaveAssetPtr=&Spot_Bird_Greenish_Warbler_02}}
 {id=WavePlayer_28, properties {soundWaveAssetPtr=&Spot_Bird_Greenish_Warbler_03}}
 {id=WavePlayer_29, properties {soundWaveAssetPtr=&Spot_Bird_Greenish_Warbler_04}}
 {id=WavePlayer_30, properties {soundWaveAssetPtr=&Spot_Bird_Greenish_Warbler_05}}
 {id=WavePlayer_31, properties {soundWaveAssetPtr=&Spot_Bird_Greenish_Warbler_01}}
 {id=WavePlayer_32, properties {soundWaveAssetPtr=&Spot_Bird_Greenish_Warbler_02}}
 {id=WavePlayer_33, properties {soundWaveAssetPtr=&Spot_Bird_Greenish_Warbler_03}}
 {id=WavePlayer_34, properties {soundWaveAssetPtr=&Spot_Bird_Greenish_Warbler_04}}
 {id=WavePlayer_35, properties {soundWaveAssetPtr=&Spot_Bird_Greenish_Warbler_05}}
 {id=WavePlayer_36, properties {soundWaveAssetPtr=&Spot_Bird_Greenish_Warbler_01}}
 {id=WavePlayer_37, properties {soundWaveAssetPtr=&Spot_Bird_Greenish_Warbler_02}}
 {id=WavePlayer_38, properties {soundWaveAssetPtr=&Spot_Bird_Greenish_Warbler_01}}
 {id=WavePlayer_39, properties {soundWaveAssetPtr=&Spot_Bird_Greenish_Warbler_02}}
 {id=WavePlayer_40, properties {soundWaveAssetPtr=&Spot_Bird_Greenish_Warbler_01}}
 {id=WavePlayer_41, properties {soundWaveAssetPtr=&Spot_Bird_Greenish_Warbler_02}}
]
properties {
 attenuationSettings=&Fort_TileAmbient_OneShot_Bird_Att
 maxDistance=8005
 soundClassObject=&Fort_Ambient_OneShots_Outdoors
 volumeMultiplier=0.2
 busSends [{
   audioBus=null
   maxSendDistance=6000
   minSendDistance=3000
   minSendLevel=0.6
   sendLevel=1
   soundSourceBus=&AmbientAudio_Exterior_SourceBus
  }]
}
