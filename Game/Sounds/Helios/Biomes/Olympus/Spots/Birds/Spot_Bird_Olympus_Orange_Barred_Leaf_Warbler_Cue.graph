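# Sound cue for Orange Barred Leaf Warbler spot bird ambience in the Olympus biome, as a quiet, attenuated outdoor one-shot.
# A quality node maps four levels onto two structurally matching branches, each feeding a modulator.
# Each branch randomly picks either a single call or a twice-looped, randomly delayed call, drawn from two warbler waves.
# Output also feeds an exterior ambient source bus.
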SoundCue /Game/Sounds/Helios/Biomes/Olympus/Spots/Birds/Spot_Bird_Olympus_Orange_Barred_Leaf_Warbler_Cue
root=QualityLevel_0
nodes [
 {id=Delay_0, children [Random_1], properties {delayMax=2}}
 {id=Delay_1, children [Random_3], properties {delayMax=2}}
 {id=Looping_0, children [Delay_0], properties {bLoopIndefinitely=false, loopCount=2}}
 {id=Looping_1, children [Delay_1], properties {bLoopIndefinitely=false, loopCount=2}}
 {id=Modulator_0, children [Random_6]}
 {id=Modulator_1, children [Random_4]}
 {id=QualityLevel_0, children [Modulator_0,Modulator_0,Modulator_1,Modulator_1]}
 {id=Random_1, children [WavePlayer_7,WavePlayer_8], properties {weights [1,1]}}
 {id=Random_3, children [WavePlayer_9, WavePlayer_10], properties {weights [1,1]}}
 {id=Random_4, children [Looping_1,Random_5], properties {weights [1,1]}}
 {id=Random_5, children [WavePlayer_11, WavePlayer_12], properties {weights [1,1]}}
 {id=Random_6, children [Looping_0,Random_7], properties {weights [1,1]}}
 {id=Random_7, children [WavePlayer_13, WavePlayer_14], properties {weights [1,1]}}
 {id=WavePlayer_10, properties {soundWaveAssetPtr=&/Spot_Bird_Orange_Barred_Leaf_Warbler_02}}
 {id=WavePlayer_11, properties {soundWaveAssetPtr=&/Spot_Bird_Orange_Barred_Leaf_Warbler_01}}
 {id=WavePlayer_12, properties {soundWaveAssetPtr=&Spot_Bird_Orange_Barred_Leaf_Warbler_02}}
 {id=WavePlayer_13, properties {soundWaveAssetPtr=&Spot_Bird_Orange_Barred_Leaf_Warbler_01}}
 {id=WavePlayer_14, properties {soundWaveAssetPtr=&Spot_Bird_Orange_Barred_Leaf_Warbler_02}}
 {id=WavePlayer_7, properties {soundWaveAssetPtr=&Spot_Bird_Orange_Barred_Leaf_Warbler_01}}
 {id=WavePlayer_8, properties {soundWaveAssetPtr=&Spot_Bird_Orange_Barred_Leaf_Warbler_02}}
 {id=WavePlayer_9, properties {soundWaveAssetPtr=&Spot_Bird_Orange_Barred_Leaf_Warbler_01}}
]
properties {
 attenuationSettings=&Fort_TileAmbient_OneShot_Bird_Att
 maxDistance=8005
 soundClassObject=&Fort_Ambient_OneShots_Outdoors
 volumeMultiplier=0.2
 busSends [{
   audioBus=null
   maxSendDistance=6000
   minSendDistance=3000
   minSendLevel=0.6
   sendLevel=1
   soundSourceBus=&AmbientAudio_Exterior_SourceBus
  }]
}
